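# Translucent UI material for Slate windows, meant to display a texture cropped to a rectangle in normalized UV coordinates.
# Normalized UVs from GetUserInterfaceUV drive a lerp between corners built from the CropRect vector parameter, and the result samples the Texture parameter.
# Only CropRect's R and B channels are read, each duplicated across both axes, so U and V share crop bounds and G and A are ignored.
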Material /Game/Slate/Window/M_CropTexture
graph {
 properties {blendMode=BLEND_Translucent, materialDomain=MD_UI, pixelDepthOffsetMode=PDOM_Legacy, usedShadingModels=Default Lit}
 nodes [
  n1 = VectorParameter() {desc="The cropping rectangle, in normalized UV coordinates", parameterName=CropRect, defaultValue=<rgba 0 0 1 1>}
  TextureSampleParameter2D(UVs=LinearInterpolate(
    A=AppendVector(A=@n1.R, B=@n1.R)
    B=AppendVector(A=@n1.B, B=@n1.B)
    Alpha=MaterialFunctionCall() {materialFunction=&GetUserInterfaceUV}.Normalized UV
   )) {constMipValue=0, desc=The texture to display, parameterName=Texture, texture=&/Maximize}
 ]
}
